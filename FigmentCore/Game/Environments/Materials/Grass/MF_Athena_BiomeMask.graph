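# Material function that returns Athena terrain biome masks for Farm, Mountain and Swamp in the TerrainColorMatch parameter group.
# World XY is offset by half of TerrainSize, negated on Y, then divided by TerrainSize to project the biome texture across the terrain.
# The GrassTextureToMatch sample uses a mip bias, with R, G and B feeding Farm, Mountain and Swamp respectively.
# The TerrainSizeHalf parameter is present but unconnected.
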MaterialFunction /FigmentCore/Game/Environments/Materials/Grass/MF_Athena_BiomeMask
graph {nodes [
  n1 = ComponentMask(None=WorldPosition().XYZ) {g=true, r=true}
  n2 = ScalarParameter() {defaultValue=306200, group=TerrainColorMatch, parameterName=TerrainSize}
  n3 = Divide(A=@n2)
  n4 = AppendVector(A=Multiply(A=@n3), B=Multiply(A=@n3) {constB=-1})
  n5 = TextureSampleParameter2D!(UVs=Divide(A=Subtract(A=@n1, B=@n4), B=@n2), Bias, Apply View MipBias) {
   constMipValue=5
   group=TerrainColorMatch
   mipValueMode=TMVM_MipBias
   parameterName=GrassTextureToMatch
   samplerType=SAMPLERTYPE_LinearColor
   texture=/FigmentCore/Game/Athena/Environments/Landscape/Textures/T_Athena_Terrain_Biome_M
  }
  FunctionOutput(None=@n5.G) {outputName=Mountain}
  FunctionOutput(None=@n5.R) {outputName=Farm}
  FunctionOutput(None=@n5.B) {outputName=Swamp}
  ScalarParameter() {defaultValue=50432, group=TerrainColorMatch, parameterName=TerrainSizeHalf}
 ]}
